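# Animation Blueprint for the Yeti medium male player body that copies the attached parent mesh's pose and runs it through the body's Control Rig.
# Each update calls CheckForBackbling to set a long-backbling flag and a float backpack-equipped value, which excludes backblings that are capes.
# The graph shows no explicit link from these variables to the Control Rig node.
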AnimBlueprint /Game/Characters/Player/Male/Medium/Bodies/M_MED_Yeti/Meshes/M_MED_Yeti_Body_AnimBP
target_skeleton=&/M_MED_Yeti_Body_Skeleton
graphs [
 AnimGraph {nodes [
   n1 = Control Rig(Source=Copy Pose From Mesh() {bUseAttachedParent=true}) {controlRigAssetReference {blueprintRigClass=&/M_MED_Yeti_Body_CtrlRig.M_MED_Yeti_Body_CtrlRig_C}}
   Output Pose(Result=@n1) {layerGroup=None}
  ]}
 EventGraph {nodes [
   Event BlueprintUpdateAnimation()
   n1 = CheckForBackbling()
   > Set isBackblingLong(isBackblingLong:bool=@n1.bIsBackblingLong) -> (Output_Get:bool)
   n2 = NOT Boolean(A=@n1.bIsBackblingACape)
   > Set isBackpackEquipped(isBackpackEquipped:double="To Float (Boolean)"(InBool=AND Boolean(A=@n1, B=@n2))) -> (Output_Get:double)
  ]}
]
variables [
 IsBackblingLong {type=boolean}
 IsBackpackEquipped {type=number}
]
